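# Sound cue for the dash sound of the IO Brute NPC's charge attack.
# A Random node picks one of four dash waves with equal weight and feeds it through a Modulator.
# It plays in the melee weapon sound class at boosted volume, with player footstep attenuation.
# A concurrency override caps simultaneous instances.
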SoundCue /NPCLibrary/Audio/IO_Brute/IOBrute_ChargeAttack_Dash_Cue
root=Modulator_0
nodes [
 {id=Modulator_0, children [Random_1]}
 {id=Random_1, children [WavePlayer_3,WavePlayer_4,WavePlayer_5,WavePlayer_6], properties {weights [1,1,1,1]}}
 {id=WavePlayer_3, properties {soundWaveAssetPtr=&/IOBrute_ChargeAttack_Dash_01}}
 {id=WavePlayer_4, properties {soundWaveAssetPtr=&/IOBrute_ChargeAttack_Dash_02}}
 {id=WavePlayer_5, properties {soundWaveAssetPtr=&/IOBrute_ChargeAttack_Dash_03}}
 {id=WavePlayer_6, properties {soundWaveAssetPtr=&/IOBrute_ChargeAttack_Dash_04}}
]
properties {
 attenuationSettings=/Game/Sounds/Attenuation/Fort_Player_Footsteps_Att
 bOverrideConcurrency=true
 maxDistance=3500
 soundClassObject=&Fort_WeaponMelee_PSM
 volumeMultiplier=2
 attenuationOverrides {falloffDistance=2700, lPFRadiusMax=3000, lPFRadiusMin=1600, attenuationShapeExtents {x=300}}
 concurrencyOverrides {maxCount=4}
 concurrencySet [null]
}
