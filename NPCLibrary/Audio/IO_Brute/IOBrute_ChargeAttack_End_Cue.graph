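# Sound cue for the end of the IO Brute NPC's charge attack.
# It randomly picks one of four equally weighted wave variations and passes it through a Modulator node.
# It uses a melee weapon sound class, borrows the player footsteps attenuation asset, boosts volume and overrides concurrency to cap simultaneous instances.
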SoundCue /NPCLibrary/Audio/IO_Brute/IOBrute_ChargeAttack_End_Cue
root=Modulator_0
nodes [
 {id=Modulator_0, children [Random_0]}
 {id=Random_0, children [WavePlayer_0,WavePlayer_1,WavePlayer_2,WavePlayer_7], properties {weights [1,1,1,1]}}
 {id=WavePlayer_0, properties {soundWaveAssetPtr=&/IOBrute_ChargeAttack_End_01}}
 {id=WavePlayer_1, properties {soundWaveAssetPtr=&/IOBrute_ChargeAttack_End_02}}
 {id=WavePlayer_2, properties {soundWaveAssetPtr=&/IOBrute_ChargeAttack_End_03}}
 {id=WavePlayer_7, properties {soundWaveAssetPtr=&/IOBrute_ChargeAttack_End_04}}
]
properties {
 attenuationSettings=/Game/Sounds/Attenuation/Fort_Player_Footsteps_Att
 bOverrideConcurrency=true
 maxDistance=3500
 soundClassObject=&Fort_WeaponMelee_PSM
 volumeMultiplier=2
 attenuationOverrides {falloffDistance=2700, lPFRadiusMax=3000, lPFRadiusMin=1600, attenuationShapeExtents {x=300}}
 concurrencyOverrides {maxCount=4}
 concurrencySet [null]
}
