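# Sound cue for the start of the IO Brute NPC's charge attack.
# A random node picks one of four equally weighted charge attack waves, which passes through a modulator for variation.
# It plays louder than default in a melee weapon sound class, using borrowed long-range scythe attenuation with HRTF spatialization.
# A global-limit concurrency asset governs simultaneous charge sounds.
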SoundCue /NPCLibrary/Audio/IO_Brute/IOBrute_ChargeAttack_Start_Cue
root=Modulator_1
nodes [
 {id=Modulator_1, children [Random_0]}
 {id=Random_0, children [WavePlayer_1,WavePlayer_2,WavePlayer_3,WavePlayer_4], properties {weights [1,1,1,1]}}
 {id=WavePlayer_1, properties {soundWaveAssetPtr=&/IOBrute_ChargeAttack_04}}
 {id=WavePlayer_2, properties {soundWaveAssetPtr=&/IOBrute_ChargeAttack_01}}
 {id=WavePlayer_3, properties {soundWaveAssetPtr=&/IOBrute_ChargeAttack_02}}
 {id=WavePlayer_4, properties {soundWaveAssetPtr=&/IOBrute_ChargeAttack_03}}
]
properties {
 attenuationSettings=/Game/Sounds/Attenuation/Fort_Weapon_Melee_Scythe_Long_Linear_Att
 maxDistance=2800
 soundClassObject=&Fort_WeaponMelee_PSM
 volumeMultiplier=2
 attenuationOverrides {spatializationAlgorithm=SPATIALIZATION_HRTF}
 concurrencySet [&/IOBrute_Charge_GlobalLimit_Conc]
}
